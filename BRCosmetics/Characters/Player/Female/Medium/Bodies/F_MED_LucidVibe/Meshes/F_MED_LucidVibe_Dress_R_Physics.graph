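# Physics asset for the right side of the LucidVibe female medium character's dress, set up for secondary motion against the lower body.
# Capsules cover the pelvis, thighs, thigh twists and calves, joined by locked constraints.
# Tiny capsules on five dress base bones lock to the pelvis as anchors.
# Only the right, right-back and right-front anchors carry larger dynamic dress capsules, attached through unlocked joints; the center anchors have none.
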PhysicsAsset /BRCosmetics/Characters/Player/Female/Medium/Bodies/F_MED_LucidVibe/Meshes/F_MED_LucidVibe_Dress_R_Physics
bodies {
 base_dress_cb_00 [{shapeType=Capsule, length=0.505, radius=0.505, shapeName=base_dress_cb_00_capsule, rotation=<rot 0 -90 0>}]
 base_dress_cf_00 [{shapeType=Capsule, length=0.505, radius=0.505, shapeName=base_dress_cf_00_capsule, rotation=<rot 0 -90 0>}]
 base_dress_r_00 [{shapeType=Capsule, length=0.505, radius=0.505, shapeName=base_dress_r_00_capsule}]
 base_dress_rb_00 [{shapeType=Capsule, length=0.505, radius=0.505, shapeName=base_dress_rb_00_capsule}]
 base_dress_rf_00 [{shapeType=Capsule, length=0.505, radius=0.505, shapeName=base_dress_rf_00_capsule}]
 calf_l [{shapeType=Capsule, length=28, radius=6, center=<-15 -2 0>, rotation=<rot 90 0 0>}]
 calf_r [{shapeType=Capsule, length=28, radius=6, center=<15 2 0>, rotation=<rot 90 0 0>}]
 dyn_dress_r_01 [{shapeType=Capsule, length=8, radius=3, shapeName=dyn_dress_r_01_capsule, center=<-2 0 0>}]
 dyn_dress_rb_01 [{shapeType=Capsule, length=8, radius=3, shapeName=dyn_dress_rb_01_capsule, center=<-2 0 0>}]
 dyn_dress_rf_01 [{shapeType=Capsule, length=8, radius=2.5, shapeName=dyn_dress_rf_01_capsule, center=<-2 0 -5>, rotation=<rot 0.0001 -0.0001 25>}]
 pelvis [{shapeType=Capsule, length=9.5, radius=10, center=<0.4878 0 0>}]
 thigh_l [{shapeType=Capsule, length=15, radius=7, center=<-14 0 0>, rotation=<rot 90 0 0>}]
 thigh_r [{shapeType=Capsule, length=15, radius=7, center=<14 0 0>, rotation=<rot 90 0 0>}]
 thigh_twist_01_l [{shapeType=Capsule, length=13, radius=6, shapeName=thigh_twist_01_l_capsule, rotation=<rot 90 0 0>}]
 thigh_twist_01_r [{shapeType=Capsule, length=13, radius=6, shapeName=thigh_twist_01_r_capsule, rotation=<rot 90 0 0>}]
}
constraints [
 {
  bone1Name=thigh_l
  bone2Name=pelvis
  swing1LimitDegrees=45
  swing1Motion=Locked
  swing2LimitDegrees=45
  swing2Motion=Locked
  twistLimitDegrees=45
  twistMotion=Locked
 }
 {
  bone1Name=calf_l
  bone2Name=thigh_l
  swing1LimitDegrees=45
  swing1Motion=Locked
  swing2LimitDegrees=45
  swing2Motion=Locked
  twistLimitDegrees=45
  twistMotion=Locked
 }
 {
  bone1Name=thigh_r
  bone2Name=pelvis
  swing1LimitDegrees=45
  swing1Motion=Locked
  swing2LimitDegrees=45
  swing2Motion=Locked
  twistLimitDegrees=45
  twistMotion=Locked
 }
 {
  bone1Name=calf_r
  bone2Name=thigh_r
  swing1LimitDegrees=45
  swing1Motion=Locked
  swing2LimitDegrees=45
  swing2Motion=Locked
  twistLimitDegrees=45
  twistMotion=Locked
 }
 {
  bone1Name=thigh_twist_01_l
  bone2Name=thigh_l
  swing1LimitDegrees=45
  swing1Motion=Locked
  swing2LimitDegrees=45
  swing2Motion=Locked
  twistLimitDegrees=45
  twistMotion=Locked
 }
 {
  bone1Name=thigh_twist_01_r
  bone2Name=thigh_r
  swing1LimitDegrees=45
  swing1Motion=Locked
  swing2LimitDegrees=45
  swing2Motion=Locked
  twistLimitDegrees=45
  twistMotion=Locked
 }
 {
  bone1Name=base_dress_cf_00
  bone2Name=pelvis
  swing1LimitDegrees=45
  swing1Motion=Locked
  swing2LimitDegrees=45
  swing2Motion=Locked
  twistLimitDegrees=45
  twistMotion=Locked
 }
 {
  bone1Name=base_dress_cb_00
  bone2Name=pelvis
  swing1LimitDegrees=45
  swing1Motion=Locked
  swing2LimitDegrees=45
  swing2Motion=Locked
  twistLimitDegrees=45
  twistMotion=Locked
 }
 {
  bone1Name=base_dress_r_00
  bone2Name=pelvis
  swing1LimitDegrees=45
  swing1Motion=Locked
  swing2LimitDegrees=45
  swing2Motion=Locked
  twistLimitDegrees=45
  twistMotion=Locked
 }
 {bone1Name=dyn_dress_r_01, bone2Name=base_dress_r_00, swing1LimitDegrees=20, swing2LimitDegrees=20, twistLimitDegrees=0}
 {
  bone1Name=base_dress_rb_00
  bone2Name=pelvis
  swing1LimitDegrees=45
  swing1Motion=Locked
  swing2LimitDegrees=45
  swing2Motion=Locked
  twistLimitDegrees=45
  twistMotion=Locked
 }
 {bone1Name=dyn_dress_rb_01, bone2Name=base_dress_rb_00, swing1LimitDegrees=40, swing2LimitDegrees=40, twistLimitDegrees=0}
 {
  bone1Name=base_dress_rf_00
  bone2Name=pelvis
  swing1LimitDegrees=45
  swing1Motion=Locked
  swing2LimitDegrees=45
  swing2Motion=Locked
  twistLimitDegrees=45
  twistMotion=Locked
 }
 {bone1Name=dyn_dress_rf_01, bone2Name=base_dress_rf_00, swing1LimitDegrees=40, swing2LimitDegrees=40, twistLimitDegrees=0}
]
