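# Physics asset for the LucidVibe female medium body, simulating secondary motion on its neckerchief and a left-side gun accessory.
# Capsules on upper spine, neck and clavicles form a rigid, locked torso frame that the accessories hang from.
# Locked base bones anchor two two-link neckerchief chains to the neck and the gun to the left clavicle.
# Dynamic links have limited swing, widening down each neckerchief chain; only the gun gets nonzero twist.
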PhysicsAsset /BRCosmetics/Characters/Player/Female/Medium/Bodies/F_MED_LucidVibe/Meshes/F_MED_LucidVibe_Neckerchief_Physics
bodies {
 base_gun_l_00 [{shapeType=Capsule, length=0.505, radius=0.505, shapeName=base_gun_l_00_capsule, rotation=<rot 0 -90 0>}]
 base_neckerchief_a_00 [{shapeType=Capsule, length=0.505, radius=0.505, shapeName=base_neckerchief_a_00_capsule, rotation=<rot 0 -90 0>}]
 base_neckerchief_b_00 [{shapeType=Capsule, length=0.505, radius=0.505, shapeName=base_neckerchief_b_00_capsule, rotation=<rot 0 -90 0>}]
 clavicle_l [{shapeType=Capsule, length=8, radius=5, shapeName=clavicle_l_capsule, center=<1.3 1.5 -0.5>, rotation=<rot 90 0 0>}]
 clavicle_r [{shapeType=Capsule, length=8, radius=5, shapeName=clavicle_r_capsule, center=<-1.3 -1.4 0.5>, rotation=<rot 90 0 180>}]
 dyn_gun_l_00 [{shapeType=Capsule, length=13.2541, radius=6.4338, shapeName=dyn_gun_l_00_capsule, center=<-3 1.7 0>, rotation=<rot 90 0 0>}]
 dyn_neckerchief_a_00 [{shapeType=Capsule, length=4, radius=1, shapeName=dyn_neckerchief_a_00_capsule, center=<2 0 0>, rotation=<rot 90 0 0>}]
 dyn_neckerchief_a_01 [{shapeType=Capsule, length=4, radius=1.5, shapeName=dyn_neckerchief_a_01_capsule, center=<2 0 0>, rotation=<rot 90 0 0>}]
 dyn_neckerchief_b_00 [{shapeType=Capsule, length=2.5, radius=1, shapeName=dyn_neckerchief_b_00_capsule, center=<2 0 0>, rotation=<rot 90 0 0>}]
 dyn_neckerchief_b_01 [{shapeType=Capsule, length=2, radius=1, shapeName=dyn_neckerchief_b_01_capsule, center=<2 0 0>, rotation=<rot 90 0 0>}]
 neck_01 [{shapeType=Capsule, length=4.4565, radius=4.5, shapeName=neck_01_capsule, center=<0 0.3581 0>, rotation=<rot 90 0 0>}]
 spine_04 [{shapeType=Capsule, length=5.5, radius=10, center=<3 0.7749 0>}]
 spine_05 [{shapeType=Capsule, length=13.5, radius=6.8, center=<2.4607 0.1295 0>}]
}
constraints [
 {
  bone1Name=spine_05
  bone2Name=spine_04
  swing1LimitDegrees=45
  swing1Motion=Locked
  swing2LimitDegrees=45
  swing2Motion=Locked
  twistLimitDegrees=45
  twistMotion=Locked
 }
 {
  bone1Name=neck_01
  bone2Name=spine_05
  swing1LimitDegrees=45
  swing1Motion=Locked
  swing2LimitDegrees=45
  swing2Motion=Locked
  twistLimitDegrees=45
  twistMotion=Locked
 }
 {
  bone1Name=base_gun_l_00
  bone2Name=clavicle_l
  swing1LimitDegrees=45
  swing1Motion=Locked
  swing2LimitDegrees=45
  swing2Motion=Locked
  twistLimitDegrees=45
  twistMotion=Locked
 }
 {bone1Name=dyn_gun_l_00, bone2Name=base_gun_l_00, swing1LimitDegrees=15, swing2LimitDegrees=10, twistLimitDegrees=10}
 {
  bone1Name=base_neckerchief_a_00
  bone2Name=neck_01
  swing1LimitDegrees=45
  swing1Motion=Locked
  swing2LimitDegrees=45
  swing2Motion=Locked
  twistLimitDegrees=45
  twistMotion=Locked
 }
 {bone1Name=dyn_neckerchief_a_00, bone2Name=base_neckerchief_a_00, swing1LimitDegrees=20, swing2LimitDegrees=20, twistLimitDegrees=0}
 {bone1Name=dyn_neckerchief_a_01, bone2Name=dyn_neckerchief_a_00, swing1LimitDegrees=30, swing2LimitDegrees=30, twistLimitDegrees=0}
 {
  bone1Name=base_neckerchief_b_00
  bone2Name=neck_01
  swing1LimitDegrees=45
  swing1Motion=Locked
  swing2LimitDegrees=45
  swing2Motion=Locked
  twistLimitDegrees=45
  twistMotion=Locked
 }
 {bone1Name=dyn_neckerchief_b_00, bone2Name=base_neckerchief_b_00, swing1LimitDegrees=20, swing2LimitDegrees=20, twistLimitDegrees=0}
 {bone1Name=dyn_neckerchief_b_01, bone2Name=dyn_neckerchief_b_00, swing1LimitDegrees=30, swing2LimitDegrees=30, twistLimitDegrees=0}
 {
  bone1Name=clavicle_l
  bone2Name=spine_05
  swing1LimitDegrees=45
  swing1Motion=Locked
  swing2LimitDegrees=45
  swing2Motion=Locked
  twistLimitDegrees=45
  twistMotion=Locked
 }
 {
  bone1Name=clavicle_r
  bone2Name=spine_05
  swing1LimitDegrees=45
  swing1Motion=Locked
  swing2LimitDegrees=45
  swing2Motion=Locked
  twistLimitDegrees=45
  twistMotion=Locked
 }
]
