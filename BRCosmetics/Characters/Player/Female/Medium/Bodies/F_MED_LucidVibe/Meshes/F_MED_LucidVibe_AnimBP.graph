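# Animation Blueprint for the LucidVibe female body cosmetic, adding secondary motion to its neckerchief and dress.
# Body pose and head-mesh facial bones blend into a Control Rig driven by crouch, jump, move and aim states, then a gun-bone constraint.
# RigidBody physics swaps neckerchief assets in the front end and simulates both dress sides, followed by AnimDynamics and backpack/cape offsets.
# The EventGraph updates those state values and computes gravity and sim-space settings.
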AnimBlueprint /BRCosmetics/Characters/Player/Female/Medium/Bodies/F_MED_LucidVibe/Meshes/F_MED_LucidVibe_AnimBP
target_skeleton=&/F_MED_LucidVibe_Skeleton
graphs [
 AnimGraph {nodes [
   n1 = "Constraint - base_neckerchief_b_00 (PPPPP)"!(
    ComponentPose
    Alpha=1
    ConstraintWeights_0=0.6501
    ConstraintWeights_1:float=0.1471
    ConstraintWeights_2:float=0.1471
    ConstraintWeights_3:float=0.0553
    ConstraintWeights_4:float=0.0004
   ) -> (Pose) {
    alphaBoolBlend {blendOption=Linear}
    boneToModify {boneName=base_neckerchief_b_00}
    constraintSetup [
     {targetBone {boneName=neck_01}}
     {targetBone {boneName=deform_pec_aim_l}}
     {targetBone {boneName=deform_pec_aim_r}}
     {targetBone {boneName=neck_02}}
     {targetBone {boneName=head}}
    ]
    constraintWeights [1,1,1,1,1]
   }
   "Constraint - base_neckerchief_a_00 (PPPPP)"!(
    ComponentPose=@n1
    Alpha=1
    ConstraintWeights_0=0.6501
    ConstraintWeights_1:float=0.1471
    ConstraintWeights_2:float=0.1471
    ConstraintWeights_3:float=0.0553
    ConstraintWeights_4:float=0.0004
   ) -> (Pose) {
    alphaBoolBlend {blendOption=Linear}
    boneToModify {boneName=base_neckerchief_a_00}
    constraintSetup [
     {targetBone {boneName=neck_01}}
     {targetBone {boneName=deform_pec_aim_l}}
     {targetBone {boneName=deform_pec_aim_r}}
     {targetBone {boneName=neck_02}}
     {targetBone {boneName=head}}
    ]
    constraintWeights [1,1,1,1,1]
   }
   n2 = RigidBody!(
    ComponentPose=Local To Component(LocalPose=Use cached pose 'constraint setup'())
    bAlphaBoolEnabled:bool=NOT Boolean(A=Get bIsInFrontEnd())
    OverrideWorldGravity:Vector="0,0,0"
    ExternalForce
   ) -> (Pose) {
    alphaInputType=Bool
    bOverrideWorldGravity=true
    overridePhysicsAsset=&/F_MED_LucidVibe_Neckerchief_Physics
    simulationSpace=BaseBoneSpace
    alphaBoolBlend {blendOption=Linear}
    baseBoneRef {boneName=spine_05}
    simSpaceSettings {
     dampingAlpha=0
     maxAngularAcceleration=40
     maxAngularVelocity=8
     maxLinearAcceleration=3500
     maxLinearVelocity=700
     worldAlpha=1
     externalLinearDragV=<1 1 1>
    }
   }
   n3 = RigidBody(ComponentPose=@n2, bAlphaBoolEnabled=Get bIsInFrontEnd(), OverrideWorldGravity="0,0,0") {
    alphaInputType=Bool
    bOverrideWorldGravity=true
    overridePhysicsAsset=&/F_MED_LucidVibe_Neckerchief_FrontEnd_Physics
    simulationSpace=BaseBoneSpace
    alphaBoolBlend {blendOption=Linear}
    baseBoneRef {boneName=spine_05}
    simSpaceSettings {
     dampingAlpha=0
     maxAngularAcceleration=40
     maxAngularVelocity=8
     maxLinearAcceleration=3500
     maxLinearVelocity=700
     worldAlpha=1
     externalLinearDragV=<1 1 1>
    }
   }
   n4 = RigidBody!(ComponentPose=@n3, Alpha:float, OverrideWorldGravity="0,0,0", ExternalForce) -> (Pose) {
    bOverrideWorldGravity=true
    overridePhysicsAsset=&/F_MED_LucidVibe_Dress_R_Physics
    simulationSpace=BaseBoneSpace
    alphaBoolBlend {blendOption=Linear}
    baseBoneRef {boneName=thigh_r}
    simSpaceSettings {
     dampingAlpha=0
     maxAngularAcceleration=40
     maxAngularVelocity=8
     maxLinearAcceleration=3500
     maxLinearVelocity=700
     worldAlpha=1
     externalLinearDragV=<1 1 1>
    }
   }
   n5 = RigidBody(ComponentPose=@n4, OverrideWorldGravity="0,0,0") {
    bOverrideWorldGravity=true
    overridePhysicsAsset=&/F_MED_LucidVibe_Dress_L_Physics
    simulationSpace=BaseBoneSpace
    alphaBoolBlend {blendOption=Linear}
    baseBoneRef {boneName=thigh_l}
    simSpaceSettings {
     dampingAlpha=0
     maxAngularAcceleration=40
     maxAngularVelocity=8
     maxLinearAcceleration=3500
     maxLinearVelocity=700
     worldAlpha=1
     externalLinearDragV=<0.5 0.5 1>
    }
   }
   n6 = Anim Dynamics - Bone: dyn_dress_cb_00(ComponentPose=@n5, Alpha=1, ExternalForce="0,0,0") {
    angularDampingOverride=1
    angularSpringConstant=20000
    bAngularSpring=true
    bOverrideAngularDamping=true
    bUseGravityOverride=true
    gravityScale=0.35
    simulationSpace=BoneRelative
    alphaBoolBlend {blendOption=Linear}
    boundBone {boneName=dyn_dress_cb_00}
    componentLinearAccScale=<1 1 1>
    componentLinearVelScale=<1 1 1>
    gravityOverride {z=-980}
    physicsBodyDefinitions [{
      boundBone {boneName=dyn_dress_cb_00}
      boxExtents=<20 30 30>
      constraintSetup {angularLimitsMax=<15 15 15>, angularLimitsMin=<-15 -15 -15>, angularTarget=<1 0 0>}
      localJointOffset=<4 0 0>
     }]
    relativeSpaceBone {boneName=root}
   }
   n7 = "Transform (Modify) Bone - Bone: attach_backpack"(ComponentPose=@n6, Alpha=1, Translation="0,4,0", Rotation="0,0,0", Scale="1,1,1") {
    translationMode=BMM_Additive
    translationSpace=BCS_BoneSpace
    alphaBoolBlend {blendOption=Linear}
    boneToModify {boneName=attach_backpack}
    translation {y=4}
   }
   n8 = "Transform (Modify) Bone - Bone: attach_cape"(ComponentPose=@n7, Alpha=1, Translation="0,3,0", Rotation="0,0,0", Scale="1,1,1") {
    translationMode=BMM_Additive
    translationSpace=BCS_BoneSpace
    alphaBoolBlend {blendOption=Linear}
    boneToModify {boneName=attach_cape}
    translation {y=3}
   }
   Output Pose(Result=Component To Local(ComponentPose=@n8)) {layerGroup=None}
   n9 = Control Rig!(
    Source=Use cached pose 'FaceAttach'()
    Alpha
    notCrouching:bool=NOT Boolean(A=Get Crouching() -> (Crouching:bool))
    crouchmoving_alpha:double=Get crouchMoving_alpha() -> (crouchMoving_alpha:double)
    isjumping_alpha:double=Get isjumping_alpha() -> (isjumping_alpha:double)
    moving2D_alpha:double=Get moving2D_alpha() -> (moving2D_alpha:double)
    aim_alpha:double=Get aimAlpha() -> (aimAlpha:double)
    RBAN off:bool=NOT Boolean(A=Get bRigidBodyEnabled())
   ) -> (Pose) {
    alphaBoolBlend {blendOption=Linear}
    controlRigAssetReference {blueprintRigClass=&/F_MED_LucidVibe_CtrlRig.F_MED_LucidVibe_CtrlRig_C}
   }
   n10 = "Constraint - base_gun_l_00 (PPPPPPPPPP)"!(
    ComponentPose=Local To Component(LocalPose=@n9)
    Alpha=1
    ConstraintWeights_0=0.5272
    ConstraintWeights_1:float=0.1412
    ConstraintWeights_2:float=0.1036
    ConstraintWeights_3:float=0.0764
    ConstraintWeights_4:float=0.0668
    ConstraintWeights_5:float=0.047
    ConstraintWeights_6:float=0.0209
    ConstraintWeights_7:float=0.0142
    ConstraintWeights_8:float=0.0019
    ConstraintWeights_9:float=0.0007
   ) -> (Pose) {
    alphaBoolBlend {blendOption=Linear}
    boneToModify {boneName=base_gun_l_00}
    constraintSetup [
     {targetBone {boneName=spine_04}}
     {targetBone {boneName=spine_02}}
     {targetBone {boneName=clavicle_l}}
     {targetBone {boneName=spine_03}}
     {targetBone {boneName=spine_05}}
     {targetBone {boneName=deform_pec_l}}
     {targetBone {boneName=spine_01}}
     {targetBone {boneName=pelvis}}
     {targetBone {boneName=upperarm_twist_01_l}}
     {targetBone {boneName=deform_upperarm_fr_l}}
    ]
    constraintWeights [1,1,1,1,1,1,1,1,1,1]
   }
   Save cached pose 'constraint setup'(Pose=Component To Local(ComponentPose=@n10)) {cacheName=constraint setup}
   n11 = Layered blend per bone(
    BasePose=Copy Pose From Mesh() {bUseAttachedParent=true}
    BlendPoses_0=Copy Pose From Mesh(SourceMeshComponent="Get Face Mesh "() -> ("Face Mesh ":Skeletal Mesh Component Object Reference)) {bUseAttachedParent=true}
   ) {layerSetup [{branchFilters [
       {boneName=FACIAL_C_FacialRoot}
       {boneName=FACIAL_C_Jaw}
       {boneName=FACIAL_C_Jawline}
       {boneName=FACIAL_R_Jawline}
       {boneName=FACIAL_L_Jawline}
       {boneName=FACIAL_R_UnderChin}
       {boneName=FACIAL_L_UnderChin}
       {boneName=FACIAL_C_UnderChin}
      ]}]}
   Save cached pose 'FaceAttach'(Pose=@n11) {cacheName=FaceAttach}
  ]}
 EventGraph {nodes [
   Event BlueprintUpdateAnimation()
   > n1 = Gravity Override State Function DEPRECATED(
    CharacterPartAnimInstance=Self-Reference()
    GameTimeInSeconds=GetGameTimeInSeconds()
    GravityOverrideParams=Get Gravity Override Params()
    RBAN Params=Get RBAN Params()
   )
   > Set GravityOverride(GravityOverride=@n1)
   > Set Sim Space Settings(Sim Space Settings:Sim Space Settings Structure=Break RBANSetting Struct!(RBANSettingStruct:"RBANSetting Struct Structure (by ref)"=@n1.RBAN SettingsOut) -> (
     ComponentLinearAccScale_28_#1:Vector=~
     ComponentLinearVelScale_31_#2:Vector=~
     SimSpaceSettings_34_#3:Sim Space Settings Structure
    ) {is=title K2Node_BreakStruct}.SimSpaceSettings_34_#3) -> (Output_Get:Sim Space Settings Structure)
   > n2 = Sequence!(execute) -> (then_0, then_1, then_2:Exec)
   n3 = Get bIsMoving2D()
   n4 = Branch(execute=@n2.then_2, Condition=@n3)
   > Set moving2D_alpha(moving2D_alpha:double=1) -> (Output_Get:double)
   Set moving2D_alpha(execute=@n4.else, moving2D_alpha=0)
   n5 = Branch(execute=@n2.then_0, Condition=AND Boolean(A=Get Crouching(), B=@n3))
   > Set crouchMoving_alpha(crouchMoving_alpha:double=1) -> (Output_Get:double)
   Set crouchMoving_alpha(execute=@n5.else, crouchMoving_alpha=0)
   n6 = Branch(execute=@n2.then_1, Condition=Get bIsJumping())
   > Set isjumping_alpha(isjumping_alpha:double=1) -> (Output_Get:double)
   Set isjumping_alpha(execute=@n6.else, isjumping_alpha=0)
   Event OnBeginCrouching()
   > Set Crouching(Crouching:bool=true) -> (Output_Get:bool)
   Event OnEndCrouching()
   > Set Crouching(Crouching=false)
   Event OnCharacterCustomizationLoaded()
   > n7 = Cast To PlayerPawn_Athena(Object=TryGetPawnOwner())
   > "Set Face Mesh "("Face Mesh ":Skeletal Mesh Component Object Reference=GetSkeletalMeshForPartType(self=@n7, PartType=Head)) -> (Output_Get:Skeletal Mesh Component Object Reference)
   Event OnBeginTargeting() {is=title K2Node_Event}
   > Set aimAlpha(aimAlpha:double=1) -> (Output_Get:double)
   Event OnEndTargeting() {is=title K2Node_Event}
   > Set aimAlpha(aimAlpha=0)
  ]}
]
variables [
 aimAlpha {type=number}
 Crouching {type=boolean}
 crouchMoving_alpha {type=number}
 "Face Mesh " {type=SkeletalMeshComponent}
 Gravity Override Params {type=GravityOverrideParamsStruct}
 GravityOverride {type=Vector}
 isjumping_alpha {type=number}
 moving2D_alpha {type=number}
 RBAN Params {
  type=RBANParamsStruct
  default {
   dBNO {simSpaceSettings {
     maxAngularAcceleration=1000
     maxAngularVelocity=3
     maxLinearAcceleration=300
     maxLinearVelocity=700
     velocityScaleZ=0
     externalLinearDragV=<0 0.5 0>
    }}
   dBNO_Carried {simSpaceSettings {
     maxAngularAcceleration=1000
     maxAngularVelocity=3
     maxLinearAcceleration=300
     maxLinearVelocity=700
     velocityScaleZ=0
     externalLinearDragV=<0 0.5 0>
    }}
   emoteOrMelee {simSpaceSettings {
     maxAngularAcceleration=1000
     maxAngularVelocity=3
     maxLinearAcceleration=300
     maxLinearVelocity=700
     velocityScaleZ=0
     externalLinearDragV=<0 0.5 0>
    }}
   front_End {simSpaceSettings {
     maxAngularAcceleration=1000
     maxAngularVelocity=3
     maxLinearAcceleration=300
     maxLinearVelocity=700
     velocityScaleZ=0
     externalLinearDragV=<0 0.5 0>
    }}
   onGround_Crouch_Moving {simSpaceSettings {
     maxAngularAcceleration=1000
     maxAngularVelocity=3
     maxLinearAcceleration=300
     maxLinearVelocity=700
     velocityScaleZ=0
     externalLinearDragV=<0 0.5 0>
    }}
   onGround_Falling {simSpaceSettings {
     maxAngularAcceleration=1000
     maxAngularVelocity=3
     maxLinearAcceleration=300
     maxLinearVelocity=700
     velocityScaleZ=0
     externalLinearDragV=<0 0.5 0>
    }}
   onGround_Moving {simSpaceSettings {
     maxAngularAcceleration=1000
     maxAngularVelocity=3
     maxLinearAcceleration=300
     maxLinearVelocity=700
     velocityScaleZ=0
     externalLinearDragV=<0 0.5 0>
    }}
   onGround_Standing {simSpaceSettings {
     maxAngularAcceleration=1000
     maxAngularVelocity=3
     maxLinearAcceleration=300
     maxLinearVelocity=700
     velocityScaleZ=0
     externalLinearDragV=<0 0.5 0>
    }}
   rail_Grinding {simSpaceSettings {
     maxAngularAcceleration=1000
     maxAngularVelocity=3
     maxLinearAcceleration=300
     maxLinearVelocity=700
     velocityScaleZ=0
     worldAlpha=0.1
     externalLinearDragV=<0 0.5 0>
    }}
   riding_Moving {simSpaceSettings {
     maxAngularAcceleration=1000
     maxAngularVelocity=3
     maxLinearAcceleration=300
     maxLinearVelocity=700
     velocityScaleZ=0
     externalLinearDragV=<0 0.5 0>
    }}
   riding_Moving_Targeting {simSpaceSettings {
     maxAngularAcceleration=1000
     maxAngularVelocity=3
     maxLinearAcceleration=300
     maxLinearVelocity=700
     velocityScaleZ=0
     externalLinearDragV=<0 0.5 0>
    }}
   riding_Standing {simSpaceSettings {
     maxAngularAcceleration=1000
     maxAngularVelocity=3
     maxLinearAcceleration=300
     maxLinearVelocity=700
     velocityScaleZ=0
     externalLinearDragV=<0 0.5 0>
    }}
   skydive_DivingDown {simSpaceSettings {
     maxAngularAcceleration=1000
     maxAngularVelocity=3
     maxLinearAcceleration=300
     maxLinearVelocity=700
     velocityScaleZ=0
     externalLinearDragV=<0 0.5 0>
    }}
   skydive_FloatingDown {
    componentLinearAccScale=<0 0.5 0>
    componentLinearVelScale=<0 0.5 0>
    simSpaceSettings {
     maxAngularAcceleration=1000
     maxAngularVelocity=3
     maxLinearAcceleration=300
     maxLinearVelocity=700
     velocityScaleZ=0
     externalLinearDragV=<0 0.5 0>
    }
   }
   skydive_GoingUp {simSpaceSettings {
     maxAngularAcceleration=1000
     maxAngularVelocity=3
     maxLinearAcceleration=300
     maxLinearVelocity=700
     velocityScaleZ=0
     externalLinearDragV=<0 0.5 0>
    }}
   skydive_Parachuting {
    componentLinearAccScale=<0 0.1 0.3>
    componentLinearVelScale=<0 0.1 0.3>
    simSpaceSettings {
     maxAngularAcceleration=1000
     maxAngularVelocity=3
     maxLinearAcceleration=300
     maxLinearVelocity=700
     velocityScaleZ=0
     externalLinearDragV=<0 0.5 0>
    }
   }
   skydive_Surfing {simSpaceSettings {
     maxAngularAcceleration=1000
     maxAngularVelocity=3
     maxLinearAcceleration=300
     maxLinearVelocity=700
     velocityScaleZ=0
     externalLinearDragV=<0 0.5 0>
    }}
   swimming {simSpaceSettings {
     maxAngularAcceleration=1000
     maxAngularVelocity=3
     maxLinearAcceleration=300
     maxLinearVelocity=700
     velocityScaleZ=0
     externalLinearDragV=<0 0.5 0>
    }}
   tactical_Sprint {simSpaceSettings {
     maxAngularAcceleration=1000
     maxAngularVelocity=3
     maxLinearAcceleration=300
     maxLinearVelocity=700
     velocityScaleZ=0
     externalLinearDragV=<0 0.5 0>
    }}
   ziplining {simSpaceSettings {
     maxAngularAcceleration=1000
     maxAngularVelocity=3
     maxLinearAcceleration=300
     maxLinearVelocity=700
     velocityScaleZ=0
     externalLinearDragV=<0 0.5 0>
    }}
  }
 }
 Sim Space Settings {
  type=SimSpaceSettings
  default {
   dampingAlpha=1
   maxAngularAcceleration=10000
   maxAngularVelocity=10000
   maxLinearAcceleration=10000
   maxLinearVelocity=10000
   velocityScaleZ=1
   worldAlpha=0
   externalLinearDragV=<0 0 0>
  }
 }
]
