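# Physics asset for the F_MED_LucidVibe face accessory mesh, defining jointed capsule bodies for the bun, upswept hair, fringe, right hair strand and both earrings.
# Each part has a base anchor locked in swing and twist to the head capsule, with swing-limited dyn bodies hanging off it.
# The fringe forks into two two-link chains and the strand is a three-link chain, their primary swing limits widening toward the tips.
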PhysicsAsset /BRCosmetics/Characters/Player/Female/Medium/Bodies/F_MED_LucidVibe/Meshes/Parts/F_MED_LucidVibe_FaceAcc_Physics
bodies {
 base_bun_c_00 [{shapeType=Capsule, length=0.505, radius=0.505, shapeName=base_bun_c_00_capsule, rotation=<rot 0 -90 0>}]
 base_earring_l_00 [{shapeType=Capsule, length=0.505, radius=0.505, shapeName=base_earring_l_00_capsule, rotation=<rot 0 -90 0>}]
 base_earring_r_00 [{shapeType=Capsule, length=0.505, radius=0.505, shapeName=base_earring_r_00_capsule, rotation=<rot 0 -90 0>}]
 base_fringe_l_00 [{shapeType=Capsule, length=0.505, radius=0.505, shapeName=base_fringe_l_00_capsule, rotation=<rot 0 -90 0>}]
 base_hairStrand_r_00 [{shapeType=Capsule, length=0.505, radius=0.505, shapeName=base_hairStrand_r_00_capsule, rotation=<rot 0 -90 0>}]
 base_hairUp_c_00 [{shapeType=Capsule, length=0.505, radius=0.505, shapeName=base_hairUp_c_00_capsule, rotation=<rot 0 -90 0>}]
 dyn_bun_c_00 [{shapeType=Capsule, length=0.5, radius=5, shapeName=dyn_bun_c_00_capsule, center=<2 0 0>, rotation=<rot 90 0 0>}]
 dyn_earring_l_00 [{shapeType=Capsule, length=2, radius=3, shapeName=dyn_earring_l_00_capsule, center=<1 0 0>, rotation=<rot 90 0 0>}]
 dyn_earring_r_00 [{shapeType=Capsule, length=1, radius=1, shapeName=dyn_earring_r_00_capsule, center=<1 0 0>, rotation=<rot 90 0 0>}]
 dyn_fringe_a_01 [{shapeType=Capsule, length=5, radius=1, shapeName=dyn_fringe_a_01_capsule, center=<2 0 0>, rotation=<rot 90 0 0>}]
 dyn_fringe_a_02 [{shapeType=Capsule, length=5, radius=1, shapeName=dyn_fringe_a_02_capsule, center=<2 0 0>, rotation=<rot 90 0 0>}]
 dyn_fringe_b_01 [{shapeType=Capsule, length=5, radius=1, shapeName=dyn_fringe_b_01_capsule, center=<2 0 0>, rotation=<rot 90 0 0>}]
 dyn_fringe_b_02 [{shapeType=Capsule, length=5, radius=1, shapeName=dyn_fringe_b_02_capsule, center=<2 0 0>, rotation=<rot 90 0 0>}]
 dyn_fringe_l_00 [{shapeType=Capsule, length=5, radius=1, shapeName=dyn_fringe_l_00_capsule, center=<2 0 0>, rotation=<rot 90 0 0>}]
 dyn_hairStrand_r_00 [{shapeType=Capsule, length=3.5, radius=0.5, shapeName=dyn_hairStrand_r_00_capsule, center=<1 -0.3 0>, rotation=<rot 90 0 0>}]
 dyn_hairStrand_r_01 [{shapeType=Capsule, length=3.5, radius=0.5, shapeName=dyn_hairStrand_r_01_capsule, center=<1 0 0>, rotation=<rot 90 0 0>}]
 dyn_hairStrand_r_02 [{shapeType=Capsule, length=3.5, radius=0.5, shapeName=dyn_hairStrand_r_02_capsule, center=<1 0 0>, rotation=<rot 90 0 0>}]
 dyn_hairUp_c_00 [{shapeType=Capsule, length=1.5, radius=1, shapeName=dyn_hairUp_c_00_capsule, center=<-1 0 0>, rotation=<rot 0 0.0002 0>}]
 head [{shapeType=Capsule, length=2.5, radius=6.5, center=<4.3987 2 0.0734>, rotation=<rot 90 0 0>}]
}
constraints [
 {bone1Name=dyn_hairStrand_r_00, bone2Name=base_hairStrand_r_00, swing1LimitDegrees=0, swing2LimitDegrees=5, twistLimitDegrees=0}
 {
  bone1Name=base_hairStrand_r_00
  bone2Name=head
  swing1LimitDegrees=45
  swing1Motion=Locked
  swing2LimitDegrees=45
  swing2Motion=Locked
  twistLimitDegrees=45
  twistMotion=Locked
 }
 {
  bone1Name=base_fringe_l_00
  bone2Name=head
  swing1LimitDegrees=45
  swing1Motion=Locked
  swing2LimitDegrees=45
  swing2Motion=Locked
  twistLimitDegrees=45
  twistMotion=Locked
 }
 {bone1Name=dyn_fringe_l_00, bone2Name=base_fringe_l_00, swing1LimitDegrees=5, swing2LimitDegrees=0, twistLimitDegrees=0}
 {bone1Name=dyn_fringe_a_01, bone2Name=dyn_fringe_l_00, swing1LimitDegrees=5, swing2LimitDegrees=3, twistLimitDegrees=0}
 {bone1Name=dyn_fringe_a_02, bone2Name=dyn_fringe_a_01, swing1LimitDegrees=8, swing2LimitDegrees=5, twistLimitDegrees=0}
 {bone1Name=dyn_fringe_b_01, bone2Name=dyn_fringe_l_00, swing1LimitDegrees=5, swing2LimitDegrees=3, twistLimitDegrees=0}
 {bone1Name=dyn_fringe_b_02, bone2Name=dyn_fringe_b_01, swing1LimitDegrees=8, swing2LimitDegrees=5, twistLimitDegrees=0}
 {bone1Name=dyn_bun_c_00, bone2Name=base_bun_c_00, swing1LimitDegrees=25, swing2LimitDegrees=30, twistLimitDegrees=5}
 {
  bone1Name=base_bun_c_00
  bone2Name=head
  swing1LimitDegrees=45
  swing1Motion=Locked
  swing2LimitDegrees=45
  swing2Motion=Locked
  twistLimitDegrees=45
  twistMotion=Locked
 }
 {
  bone1Name=base_earring_l_00
  bone2Name=head
  swing1LimitDegrees=45
  swing1Motion=Locked
  swing2LimitDegrees=45
  swing2Motion=Locked
  twistLimitDegrees=45
  twistMotion=Locked
 }
 {bone1Name=dyn_earring_l_00, bone2Name=base_earring_l_00, swing1LimitDegrees=10, swing2LimitDegrees=30, twistLimitDegrees=0}
 {
  bone1Name=base_earring_r_00
  bone2Name=head
  swing1LimitDegrees=45
  swing1Motion=Locked
  swing2LimitDegrees=45
  swing2Motion=Locked
  twistLimitDegrees=45
  twistMotion=Locked
 }
 {bone1Name=dyn_earring_r_00, bone2Name=base_earring_r_00, swing1LimitDegrees=20, swing2LimitDegrees=30, twistLimitDegrees=0}
 {bone1Name=dyn_hairUp_c_00, bone2Name=base_hairUp_c_00, swing1LimitDegrees=25, swing2LimitDegrees=25, twistLimitDegrees=0}
 {
  bone1Name=base_hairUp_c_00
  bone2Name=head
  swing1LimitDegrees=45
  swing1Motion=Locked
  swing2LimitDegrees=45
  swing2Motion=Locked
  twistLimitDegrees=45
  twistMotion=Locked
 }
 {bone1Name=dyn_hairStrand_r_01, bone2Name=dyn_hairStrand_r_00, swing1LimitDegrees=20, swing2LimitDegrees=10, twistLimitDegrees=0}
 {bone1Name=dyn_hairStrand_r_02, bone2Name=dyn_hairStrand_r_01, swing1LimitDegrees=30, swing2LimitDegrees=5, twistLimitDegrees=0}
]
